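# Translucent deferred decal material for the BigSplat set, with bug decal textures as defaults.
# It exposes Diffuse, Normal and Specular masks texture parameters, sampled at UVs divided by a Texture tiling scalar, so higher values enlarge the pattern.
# A second Diffuse sampler uses default UVs, and the file shows no connections from any sampler to material outputs.
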Material /Hermes/Sets/BigSplat/Materials/Decals/M_BigSplat_Decal
graph {
 properties {bAutomaticallySetUsageInEditor=false, blendMode=BLEND_Translucent, materialDomain=MD_DeferredDecal, usedShadingModels=Default Lit}
 nodes [
  TextureSampleParameter2D() {group=Textures, parameterName=Diffuse, texture=/Hermes/Sets/BigSplat/Textures/Decals/BigSplat_BugDecal_A_D}
  n1 = Divide(A=TextureCoordinate(), B=ScalarParameter() {defaultValue=1, parameterName=Texture tiling})
  TextureSampleParameter2D(UVs=@n1) {group=Textures, parameterName=Normal, samplerType=SAMPLERTYPE_Normal, texture=/Hermes/Sets/BigSplat/Textures/Decals/BigSplat_BugDecal_A_N}
  TextureSampleParameter2D(UVs=@n1) {
   group=Textures
   parameterName=Specular masks
   samplerType=SAMPLERTYPE_Masks
   texture=/Hermes/Sets/BigSplat/Textures/Decals/BigSplat_BugDecal_A_S
  }
  TextureSampleParameter2D(UVs=@n1) {group=Textures, parameterName=Diffuse, texture=&BigSplat_BugDecal_A_D}
 ]
}
